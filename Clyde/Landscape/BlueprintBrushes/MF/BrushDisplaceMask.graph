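# Material function for landscape Blueprint brushes that turns a UV distance into a displacement mask.
# The DisplacementMaskParams vector supplies offset, width and power: distance is offset, divided by width, centred on 0.5, saturated and shaped by SCurve.
# Its alpha channel blends from a constant 1 (no masking, the default) to that curve.
# Three DisplaceMask scalar parameters are present but unconnected.
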MaterialFunction /Clyde/Landscape/BlueprintBrushes/MF/BrushDisplaceMask
graph {nodes [
  n1 = VectorParameter() {desc="Offset, Width, power, disable when width=0", parameterName=DisplacementMaskParams, defaultValue=<rgba 0 0.0001 0.0001 0>}
  n2 = Add(A=FunctionInput() {bUsePreviewValueAsDefault=true, inputName=UV Distance, inputType=FunctionInput_Scalar, sortPriority=1}, B=@n1.R)
  n3 = Saturate(None=Add(A=Divide(A=@n2, B=@n1.G)) {constB=0.5})
  FunctionOutput(None=LinearInterpolate(
    B=MaterialFunctionCall(In=@n3, Power=@n1.B) {materialFunction=&SCurve}.Result
    Alpha=@n1.A
   ) {constA=1}) {outputName=Mask}
  ScalarParameter() {defaultValue=1, parameterName=DisplaceMaskWidth}
  ScalarParameter() {parameterName=DisplaceMaskOffset}
  ScalarParameter() {defaultValue=1, parameterName=DisplaceMaskPower}
 ]}
