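# LocalizeLandscapeZValue is a material function that converts a world Z value into landscape-local Z for landscape blueprint brush materials.
# It reads the landscape's Z location and Z scale from the Landmass MPC_Landscape parameter collection.
# The Landscape Local Z output subtracts the location and divides by scale, while Scale Only divides the input by scale without the offset.
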MaterialFunction /Clyde/Landscape/BlueprintBrushes/MF/LocalizeLandscapeZValue
graph {nodes [
  n1 = FunctionInput() {bUsePreviewValueAsDefault=true, inputName=World Z, inputType=FunctionInput_Scalar}
  n2 = CollectionParameter() {collection=/Landmass/Landscape/BlueprintBrushes/MPC/MPC_Landscape, parameterName=LandscapeZLocation}
  n3 = CollectionParameter() {collection=&MPC_Landscape, parameterName=LandscapeZScale}
  FunctionOutput(None=Divide(A=Subtract(A=@n1, B=@n2), B=@n3)) {outputName=Landscape Local Z}
  FunctionOutput(None=Divide(A=@n1, B=@n3)) {outputName=Scale Only, sortPriority=1}
 ]}
