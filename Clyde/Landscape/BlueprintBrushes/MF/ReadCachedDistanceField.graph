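# Material function that reads and unpacks a cached distance field height texture for Landmass landscape blueprint brushes.
# It samples at landscape texel centers plus an optional UV offset, at a fixed mip level.
# Floored red gives world distance, also normalized by landscape world size; floored green gives water height, blue gives depth, and both derive terrain height.
# Velocity packs direction in red/green fractional parts, scaled by inverted alpha; raw RGBA is also output.
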MaterialFunction /Clyde/Landscape/BlueprintBrushes/MF/ReadCachedDistanceField
graph {nodes [
  n1 = TextureSample!(
   UVs=Add(
    A=MaterialFunctionCall() {materialFunction=/Landmass/Landscape/BlueprintBrushes/MF/PointSampleLandscapeUVs}.TexelCenterUVs
    B=FunctionInput() {bUsePreviewValueAsDefault=true, inputName=UV Offset, inputType=FunctionInput_Vector2, sortPriority=1}
   )
   Tex=FunctionInput(Preview=TextureObjectParameter() {parameterName=CachedDistanceFieldHeight, texture=&Black}) {bUsePreviewValueAsDefault=true, inputName=Cached DF, inputType=FunctionInput_Texture2D}
   Level=Constant()
   Apply View MipBias
  ) {mipValueMode=TMVM_MipLevel}
  Ceil(None=@n1.G)
  n2 = Floor(None=@n1.G)
  FunctionOutput(None=Add(A=Subtract(A=@n2, B=@n1.B))) {outputName=Terrain Height, sortPriority=2}
  n3 = Floor(None=@n1.R)
  n4 = CollectionParameter() {collection=/Landmass/Landscape/BlueprintBrushes/MPC/MPC_Landscape, parameterName=WorldSizeX}
  n5 = CollectionParameter() {collection=&MPC_Landscape, parameterName=WorldSizeY}
  FunctionOutput(None=Divide(A=FunctionOutput(None=@n3) {outputName=World Distance}, B=Max(A=@n4, B=@n5))) {
   description="The distance in world space, based on a tangent of 1, aka a 45 degree angle which has a tangent or multiplier of 1.0."
   outputName=UV Distance
   sortPriority=1
  }
  FunctionOutput(None=@n2) {outputName=Water Height, sortPriority=2}
  FunctionOutput(None=@n1.B) {outputName=Water Depth, sortPriority=3}
  n6 = Frac(None=ComponentMask(None=@n1.RGB) {g=true, r=true})
  n7 = OneMinus(None=@n1.A)
  FunctionOutput(None=Multiply(A=ConstantBiasScale(None=@n6) {bias=-0.5, scale=2}, B=Multiply(A=@n7) {constB=1000})) {outputName=Velocity, sortPriority=4}
  FunctionOutput(None=@n1.RGBA) {outputName=RGBA, sortPriority=4}
 ]}
