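# Landmass landscape brush material that renders a cached distance field, packing edge distance, height and channel data into one output.
# Floored distance and landscape-local Z share two components with normalized ChannelDepth direction in their fractions; a third holds CurveRT-shaped channel depth.
# Height optionally blends in captured MeshDepth inside the shape, with depth textures sampled through two-layer curl-noise UV distortion.
# A nine-tap distance-field blur exists, but its blend alpha is zero.
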Material /Clyde/Landscape/BlueprintBrushes/Materials/CacheDistanceField
graph {
 properties {
  bAllowNegativeEmissiveColor=true
  blendMode=BLEND_AlphaComposite
  pixelDepthOffsetMode=PDOM_Legacy
  shadingModel=MSM_Unlit
  usedShadingModels=Unlit
 }
 nodes [
  n1 = Constant()
  n2 = MaterialFunctionCall!(UV Offset=@n1) {desc=Note: Z height here is at Nearest Edge, materialFunction=/Landmass/Landscape/BlueprintBrushes/MF/SampleDistanceField}
  n3 = AppendVector(A=@n2.UV Distance, B=@n2.UV Distance)
  n4 = ScalarParameter() {defaultValue=1, parameterName=Bluroffset}
  n5 = Divide(
   A=@n4
   B=AppendVector(
    A=CollectionParameter() {collection=/Landmass/Landscape/BlueprintBrushes/MPC/MPC_Landscape, parameterName=RTResX}
    B=CollectionParameter() {collection=&MPC_Landscape, parameterName=RTResY}
   )
  )
  n6 = Multiply(A=@n5, B=Constant2Vector() {r=1})
  n7 = MaterialFunctionCall(UV Offset=Add(A=@n1, B=@n6)) {materialFunction=&SampleDistanceField}
  n8 = MaterialFunctionCall(UV Offset=Add(A=@n1, B=Multiply(A=@n6) {constB=-1})) {materialFunction=&SampleDistanceField}
  n9 = Add(A=AppendVector(A=@n7.UV Distance, B=@n7.UV Distance), B=AppendVector(A=@n8.UV Distance, B=@n8.UV Distance))
  n10 = Multiply(A=@n5, B=Constant2Vector() {g=1})
  n11 = MaterialFunctionCall(UV Offset=Add(A=@n1, B=@n10)) {materialFunction=&SampleDistanceField}
  n12 = MaterialFunctionCall(UV Offset=Add(A=@n1, B=Multiply(A=@n10) {constB=-1})) {materialFunction=&SampleDistanceField}
  n13 = Add(A=AppendVector(A=@n11.UV Distance, B=@n11.UV Distance), B=AppendVector(A=@n12.UV Distance, B=@n12.UV Distance))
  n14 = Divide(
   A=@n4
   B=AppendVector(
    A=CollectionParameter() {collection=&MPC_Landscape, parameterName=RTResX}
    B=CollectionParameter() {collection=&MPC_Landscape, parameterName=RTResY}
   )
  )
  n15 = Multiply(A=@n14, B=Constant2Vector() {g=1, r=1})
  n16 = MaterialFunctionCall(UV Offset=Add(A=@n1, B=@n15)) {materialFunction=&SampleDistanceField}
  n17 = Constant2Vector() {g=1, r=-1}
  n18 = MaterialFunctionCall(UV Offset=Add(A=@n1, B=Multiply(A=@n15, B=@n17) {constB=-1})) {materialFunction=&SampleDistanceField}
  n19 = Add(A=AppendVector(A=@n16.UV Distance, B=@n16.UV Distance), B=AppendVector(A=@n18.UV Distance, B=@n18.UV Distance))
  n20 = Multiply(A=@n14, B=Constant2Vector() {g=-1, r=-1})
  n21 = MaterialFunctionCall(UV Offset=Add(A=@n1, B=@n20)) {materialFunction=&SampleDistanceField}
  n22 = Constant2Vector() {g=1, r=-1}
  n23 = MaterialFunctionCall(UV Offset=Add(A=@n1, B=Multiply(A=@n20, B=@n22) {constB=-1})) {materialFunction=&SampleDistanceField}
  n24 = Add(A=AppendVector(A=@n21.UV Distance, B=@n21.UV Distance), B=AppendVector(A=@n23.UV Distance, B=@n23.UV Distance))
  n25 = Add(A=Add(A=Add(A=@n9, B=@n13), B=Add(A=@n19, B=@n24)), B=@n3)
  n26 = MaterialFunctionCall(Float2=LinearInterpolate(A=@n3, B=Divide(A=@n25) {constB=9}) {constAlpha=0}) {materialFunction=&BreakOutFloat2Components}
  n27 = CollectionParameter() {collection=&MPC_Landscape, parameterName=WorldSizeX}
  n28 = CollectionParameter() {collection=&MPC_Landscape, parameterName=WorldSizeY}
  n29 = Floor(None=Multiply(A=@n26.R, B=Max(A=@n27, B=@n28)))
  n30 = AppendVector(A=TextureCoordinate(), B=ScalarParameter() {parameterName=Seed1})
  n31 = VectorParameter() {parameterName=Curl, defaultValue=<rgba 3 0.025 24 -0.0025>}
  n32 = Multiply(A=VectorNoise(World Position=Multiply(A=@n30, B=@n31.R)) {noiseFunction=VNF_CurlALU}, B=Divide(A=@n31.G) {constB=100})
  n33 = AppendVector(A=TextureCoordinate(), B=ScalarParameter() {parameterName=Seed2})
  n34 = Multiply(A=VectorNoise(World Position=Multiply(A=@n33, B=@n31.B)) {noiseFunction=VNF_CurlALU}, B=Divide(A=@n31.A) {constB=100})
  n35 = Add(A=ComponentMask(None=Add(A=@n32, B=@n34)) {g=true, r=true}, B=TextureCoordinate())
  n36 = TextureSampleParameter2D(UVs=@n35) {parameterName=MeshDepth, texture=/Engine/EngineResources/Black_Low.Black}
  n37 = Subtract(A=ScalarParameter() {defaultValue=50000, parameterName=SceneCaptureZ}, B=@n36.R) {constA=50000}
  n38 = Subtract(
   A=Subtract(A=ScalarParameter() {defaultValue=50000, parameterName=SceneCaptureZ}, B=ScalarParameter() {defaultValue=-50000, parameterName=MinZ})
   B=@n36.R
  )
  n39 = Multiply(A=@n2.Mask, B=Saturate(None=Ceil(None=@n38)) {desc=Areas Captured by Depth Mask}) {desc=Make sure to only use values in Mesh Depth}
  n40 = Add(
   A=LinearInterpolate(
    A=@n26.G
    B=LinearInterpolate(A=@n26.G, B=@n37, Alpha=@n39) {desc=Blend Actual Depth inside Shape}
    Alpha=ScalarParameter() {parameterName=UseMeshDepth}
   ) {desc="If does not use Depth, Z is encoded into DF"}
   B=ScalarParameter() {parameterName=ZOffset}
  )
  n41 = Floor(None=MaterialFunctionCall!(World Z=@n40) {materialFunction=/Landmass/Landscape/BlueprintBrushes/MF/LocalizeLandscapeZValue}.Landscape Local Z)
  n42 = TextureSampleParameter2D(UVs=@n35) {parameterName=ChannelDepth, texture=&Black}
  n43 = ComponentMask(None=@n42.RGB) {g=true, r=true}
  n44 = ConstantBiasScale(None=Normalize(VectorInput=@n43))
  n45 = LinearInterpolate(
   A=ScalarParameter() {defaultValue=500, parameterName=CurveChannelDepth}
   B=Subtract(A=@n36.R, B=@n42.A)
   Alpha=ScalarParameter() {parameterName=UseMeshDepth}
  )
  n46 = Add(A=@n2.World Distance, B=ScalarParameter() {parameterName=ChannelEdgeOffset})
  n47 = ScalarParameter() {defaultValue=500, parameterName=CurveRampWidth}
  n48 = Divide() {constA=0.5, constB=256}
  n49 = Clamp(None=Saturate(None=Divide(A=@n46, B=@n47)), Min=@n48, Max=OneMinus(None=@n48))
  n50 = TextureSampleParameter2D!(UVs=@n49, Level=Constant(), Apply View MipBias) {mipValueMode=TMVM_MipLevel, parameterName=CurveRT, texture=/Landmass/Landscape/Curves/DynamicCurve_RT}
  AppendVector(
   A=Add(A=AppendVector(A=@n29, B=@n41), B=Clamp(None=@n44) {maxDefault=0.99}) {constB=0}
   B=Divide(
    A=Multiply(A=@n45, B=@n50.R)
    B=CollectionParameter() {collection=&MPC_Landscape, parameterName=LandscapeZScale}
   )
  )
  Constant() {r=1}
  OneMinus(None=@n49)
  Saturate(None=Divide(A=MaterialFunctionCall(Vector 2=@n43) {materialFunction=/Engine/Functions/Engine_MaterialFunctions02/Utility/VectorLength}.V2 Length) {constB=1000})
  ScalarParameter() {defaultValue=-2048, parameterName=Zzzzz}
  ScalarParameter() {parameterName=UseBlur}
  ScalarParameter() {defaultValue=1394, parameterName=Z}
  Subtract(A=@n40) {desc=WIP Fix for 1 unit offset}
 ]
}
