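# Unlit alpha-composite material for Landmass blueprint brushes that merges water velocity, height and depth from a cached distance field into the CombinedVelocityAndHeight texture.
# Four texel-offset neighbour samples dilate water height outward, taking the masked maximum and lowering it slightly outside the centre mask.
# The centre mask, extended by dilation where existing alpha is full, blends velocity and Z-offset height into RGB.
# A separate alpha branch blends normalised water depth with the existing alpha.
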Material /Clyde/Landscape/BlueprintBrushes/Materials/CompositeAlphas
graph {
 properties {
  bAllowNegativeEmissiveColor=true
  blendMode=BLEND_AlphaComposite
  pixelDepthOffsetMode=PDOM_Legacy
  shadingModel=MSM_Unlit
  usedShadingModels=Unlit
 }
 nodes [
  n1 = MaterialFunctionCall!(Cached DF, UV Offset) {materialFunction=/Landmass/Landscape/BlueprintBrushes/MF/ReadCachedDistanceField}
  Add(
   A=CollectionParameter() {collection=/Landmass/Landscape/BlueprintBrushes/MPC/MPC_Landscape, parameterName=LandscapeZLocation}
   B=Multiply(
    A=CollectionParameter() {collection=&MPC_Landscape, parameterName=LandscapeZScale}
    B=@n1.Water Height
   )
  )
  Constant() {r=0.9999}
  n2 = TextureSampleParameter2D() {parameterName=CombinedVelocityAndHeight, texture=&Black}
  n3 = CollectionParameter() {collection=&MPC_Landscape, parameterName=RTResX}
  n4 = CollectionParameter() {collection=&MPC_Landscape, parameterName=RTResY}
  n5 = Multiply(A=ScalarParameter() {defaultValue=2, parameterName=Dilation Spread}, B=Divide(B=AppendVector(A=@n3, B=@n4)))
  n6 = Multiply(A=Multiply(A=@n5, B=Constant2Vector() {r=1})) {constB=-1}
  n7 = MaterialFunctionCall(UV Offset=@n6) {materialFunction=&ReadCachedDistanceField}
  n8 = Saturate(None=Ceil(None=Multiply(A=@n7.World Distance))) {desc=Mask}
  n9 = MaterialFunctionCall(UV Offset=@n6) {materialFunction=&ReadCachedDistanceField}
  n10 = Saturate(None=Ceil(None=Multiply(A=@n9.World Distance))) {desc=Mask}
  n11 = Max(A=LinearInterpolate(B=@n7.Water Height, Alpha=@n8) {constA=-50000}, B=LinearInterpolate(B=@n9.Water Height, Alpha=@n10) {constA=-50000})
  n12 = Multiply(A=@n5, B=Constant2Vector() {g=1})
  n13 = MaterialFunctionCall(UV Offset=@n12) {materialFunction=&ReadCachedDistanceField}
  n14 = Saturate(None=Ceil(None=Multiply(A=@n13.World Distance))) {desc=Mask}
  n15 = MaterialFunctionCall(UV Offset=Multiply(A=@n12) {constB=-1}) {materialFunction=&ReadCachedDistanceField}
  n16 = Saturate(None=Ceil(None=Multiply(A=@n15.World Distance))) {desc=Mask}
  n17 = Max(A=LinearInterpolate(B=@n13.Water Height, Alpha=@n14) {constA=-50000}, B=LinearInterpolate(B=@n15.Water Height, Alpha=@n16) {constA=-50000})
  n18 = Saturate(None=Ceil(None=Multiply(A=@n1.World Distance))) {desc=Mask}
  n19 = LinearInterpolate(A=Subtract(A=Max(A=@n11, B=@n17)) {constB=32}, B=@n1.Water Height, Alpha=@n18)
  n20 = MaterialFunctionCall!(World Z=ScalarParameter() {parameterName=ZOffset}) {materialFunction=/Landmass/Landscape/BlueprintBrushes/MF/LocalizeLandscapeZValue}
  n21 = Max(A=Max(A=@n8, B=@n10), B=Max(A=@n14, B=@n16))
  n22 = Saturate(None=Floor(None=@n2.A))
  n23 = Max(A=@n18, B=Multiply(A=Multiply(A=@n21), B=@n22)) {constB=0}
  LinearInterpolate(A=@n2.RGB, B=AppendVector(A=@n1.Velocity, B=Subtract(A=@n19, B=@n20.Scale Only)), Alpha=@n23)
  n24 = Saturate(None=Divide(A=Subtract(A=@n1.Water Depth, B=@n20.Scale Only)) {constB=1000})
  n25 = OneMinus(None=Max(A=@n24, B=OneMinus(None=@n2.A)))
  n26 = OneMinus(None=@n24)
  n27 = Multiply(A=@n23, B=@n22)
  OneMinus(None=LinearInterpolate(A=@n2.A, B=LinearInterpolate(A=@n25, B=@n26, Alpha=@n27), Alpha=@n23))
 ]
}
