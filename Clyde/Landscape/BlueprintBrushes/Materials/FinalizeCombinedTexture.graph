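# Unlit Landmass brush material that finalizes the combined water velocity and height texture for a landscape.
# A Custom node box-blurs XY velocity, averaging only neighbors with alpha below one.
# Texels still at the clear height with full alpha take offset-lowered terrain height, floored at clear height, before conversion to world Z.
# MakeFloat3 packs blurred X into both X and Y alongside that height; a separate branch computes normalized water depth over terrain.
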Material /Clyde/Landscape/BlueprintBrushes/Materials/FinalizeCombinedTexture
graph {
 properties {
  bAllowNegativeEmissiveColor=true
  blendMode=BLEND_AlphaComposite
  pixelDepthOffsetMode=PDOM_Legacy
  shadingModel=MSM_Unlit
  usedShadingModels=Unlit
 }
 nodes [
  MaterialExpressionTextureObjectParameter_0 = TextureObjectParameter() {parameterName=CombinedVelocityAndHeight, texture=&Black}
  MaterialExpressionTextureCoordinate_0 = TextureCoordinate()
  n1 = CollectionParameter() {collection=/Landmass/Landscape/BlueprintBrushes/MPC/MPC_Landscape, parameterName=RTResX}
  n2 = CollectionParameter() {collection=&MPC_Landscape, parameterName=RTResY}
  MaterialExpressionMultiply_1 = Multiply(A=Divide(B=AppendVector(A=@n1, B=@n2)), B=ScalarParameter() {defaultValue=1, parameterName=Texel Scale})
  MaterialExpressionScalarParameter_4 = ScalarParameter() {defaultValue=5, parameterName=Radius}
  n3 = Custom!(
   Tex=@MaterialExpressionTextureObjectParameter_0
   UV=@MaterialExpressionTextureCoordinate_0
   TexelSize=@MaterialExpressionMultiply_1
   Radius=@MaterialExpressionScalarParameter_4
  ) {
   code="float2 accum = 0;\r\n\r\nfloat2 curUV = UV;\r\nfloat neighbors = 0;\r\n\r\nfor (int i = -Radius; i <= Radius; i++)\r\n{\r\n\r\n\tfor(int j=-Radius; j <= Radius; j++)\r\n\t{\r\n\t\t\r\n\t\tcurUV = UV + float2(TexelSize.x * (float)i, 0) + float2(0, TexelSize.y * (float)j);\r\n\t\tfloat4 cursample = Tex.SampleLevel(TexSampler, curUV, 0);\r\n\t\t\r\n\t\tfloat depthmask = 1 - floor(cursample.a);\r\n\t\t\r\n\t\tif (depthmask == 1.0)\r\n\t\t{\r\n\t\t\taccum += cursample.xy;\r\n\t\t\tneighbors += 1;\r\n\t\t\r\n\t\t}\r\n\t\t\r\n\t}\r\n\t\r\n}\r\n\r\nif(neighbors > 0)\r\n{\r\n\taccum /= neighbors;\r\n}\r\n\r\nreturn accum.xy;"
   description=Blur Velocities
   outputType=CMOT_Float2
   inputs [
    {
     inputName=Tex
     input {expression=&self:MaterialExpressionTextureObjectParameter_0}
    }
    {inputName=UV, input {expression=&self:MaterialExpressionTextureCoordinate_0}}
    {inputName=TexelSize, input {expression=&self:MaterialExpressionMultiply_1}}
    {inputName=Radius, input {expression=&self:MaterialExpressionScalarParameter_4}}
   ]
  }
  n4 = TextureSampleParameter2D() {parameterName=CombinedVelocityAndHeight, texture=&Black}
  n5 = MaterialFunctionCall(Float3=AppendVector(A=@n3, B=@n4.B)) {materialFunction=&BreakOutFloat3Components}
  n6 = MaterialFunctionCall!(RG=TextureSampleParameter2D() {parameterName=LandscapeHeight, texture=&Black}.RGB) {materialFunction=/Landmass/Landscape/BlueprintBrushes/MF/UnpackRG8Height}
  n7 = ScalarParameter() {defaultValue=256, parameterName=Water Terrain Offset}
  n8 = Divide(
   A=Subtract(
    A=CollectionParameter() {collection=&MPC_Landscape, parameterName=WaterClearHeight}
    B=CollectionParameter() {collection=&MPC_Landscape, parameterName=LandscapeZLocation}
   )
   B=CollectionParameter() {collection=&MPC_Landscape, parameterName=LandscapeZScale}
  )
  n9 = Constant()
  n10 = Multiply(
   A=Floor(None=@n4.A)
   B=If(
    A=@n4.B
    B=CollectionParameter() {collection=&MPC_Landscape, parameterName=WaterClearHeight}
    "A > B"=@n9
    "A == B"=Constant() {r=1}
    "A < B"=@n9
   ) {equalsThreshold=1}
  )
  n11 = LinearInterpolate(A=@n5.B, B=Max(A=Subtract(A=@n6.Result, B=@n7), B=@n8), Alpha=@n10)
  n12 = CollectionParameter() {collection=&MPC_Landscape, parameterName=LandscapeZScale}
  MaterialFunctionCall(
   X=@n5.R
   Y=@n5.R
   Z=Add(
    A=Multiply(A=@n11, B=@n12)
    B=CollectionParameter() {collection=&MPC_Landscape, parameterName=LandscapeZLocation}
   )
  ) {materialFunction=&MakeFloat3}
  n13 = Max(A=LinearInterpolate(A=Subtract(A=@n4.B, B=@n6.Result), B=Subtract(A=@n8, B=@n6.Result), Alpha=@n10)) {constB=0}
  Saturate(None=Multiply(A=Divide(A=@n13) {constB=1000}, B=@n12))
  TextureSampleParameter2D() {parameterName=CombinedVelocityAndHeight, texture=&Black}
 ]
}
