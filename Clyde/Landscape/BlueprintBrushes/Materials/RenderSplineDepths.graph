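# Unlit spline-mesh material for Landscape Blueprint brushes, named for rendering spline depths, with negative emissive allowed.
# It computes world Z relative to the MPC_Landscape LandscapeZLocation, divided by a scalar.
# Separate branches blend VelA to VelB along U, scale the tangent X axis by that velocity, and shape a cross-spline falloff with SphereMask and a Pythagorean function.
# No material output connections appear, so which branch gets rendered is unclear.
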Material /Clyde/Landscape/BlueprintBrushes/Materials/RenderSplineDepths
graph {
 properties {
  bAllowNegativeEmissiveColor=true
  bUsedWithSplineMeshes=true
  pixelDepthOffsetMode=PDOM_Legacy
  shadingModel=MSM_Unlit
  usedShadingModels=Unlit
 }
 nodes [
  AppendVector()
  n1 = WorldPosition()
  Divide(
   A=Subtract(
    A=ComponentMask(None=@n1.XYZ) {b=true}
    B=CollectionParameter() {collection=/Landmass/Landscape/BlueprintBrushes/MPC/MPC_Landscape, parameterName=LandscapeZLocation}
   )
   B=ScalarParameter() {defaultValue=1000, parameterName=dfadfasdf}
  )
  n2 = LinearInterpolate(
   A=ScalarParameter() {defaultValue=128, parameterName=VelA}
   B=ScalarParameter() {defaultValue=128, parameterName=VelB}
   Alpha=ComponentMask(None=TextureCoordinate()) {r=true}
  )
  Divide(A=@n2) {constB=100}
  LinearInterpolate(
   A=ScalarParameter() {parameterName=A}
   B=ScalarParameter() {defaultValue=0.2, parameterName=B}
   Alpha=ComponentMask(None=TextureCoordinate()) {r=true}
  )
  Multiply(A=Multiply(A=@n2, B=MaterialFunctionCall() {materialFunction=&TangentBasis}.X U))
  n3 = ComponentMask(None=TextureCoordinate()) {g=true}
  n4 = Constant() {r=0.5}
  n5 = Constant() {r=0.45}
  n6 = ScalarParameter() {parameterName=Velocity Edge Amount}
  n7 = MaterialFunctionCall!(Side 1 Length=OneMinus(None=SphereMask(A=@n3, B=@n4, Radius=@n5, Hardness=@n6))) {materialFunction=/Engine/Functions/Engine_MaterialFunctions02/Math/pythagoreanTheorum}
  Multiply(A=@n7.Side 2 Length, B=@n7.Side 2 Length)
  ScalarParameter() {defaultValue=50000, parameterName=SceneCaptureZ}
  ScalarParameter() {defaultValue=0.2}
 ]
}
